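# Material function that draws an arbitrary four-vertex polygon as a signed distance field, stored with the AloeCrouton weapon's UI materials.
# UVs are remapped to -1..1 and fed with four vertex inputs into a custom HLSL node that computes edge distance and inside/outside sign.
# Outline thickness turns the distance into a stroke SDF.
# Outputs are both raw SDFs plus Fill and Outline masks, softened by inverted SmoothStep with Glow Min/Max inputs.
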MaterialFunction /AloeCroutonWeapon/Items/AloeCrouton/UI/Materials/MF_4VertexPolygon
graph {nodes [
  FunctionOutput(None=NamedRerouteUsage()) {outputName=SDF, sortPriority=3}
  FunctionOutput(None=NamedRerouteUsage()) {outputName=SDF Outline, sortPriority=4}
  n1 = NamedRerouteUsage()
  n2 = NamedRerouteUsage()
  n3 = NamedRerouteUsage()
  FunctionOutput(None=OneMinus(None=SmoothStep(Min=@n1, Max=@n2, Value=@n3))) {outputName=Fill}
  n4 = NamedRerouteUsage()
  n5 = NamedRerouteUsage()
  n6 = NamedRerouteUsage()
  FunctionOutput(None=OneMinus(None=SmoothStep(Min=@n4, Max=@n5, Value=@n6))) {outputName=Outline, sortPriority=2}
  n7 = FunctionInput(Preview=TextureCoordinate()) {
   bUsePreviewValueAsDefault=true
   description=The Position Coordinates to measure the box. Default is World Position and can be left default in most cases.
   inputName=UVs
   inputType=FunctionInput_Vector2
  }
  NamedRerouteDeclaration(None=ConstantBiasScale(None=@n7) {bias=-0.5, scale=2}) {name="UV's(-1,1)", nodeColor=<rgba 0 0 0 1>}
  NamedRerouteDeclaration(None=FunctionInput(Preview=Constant2Vector() {g=-1, r=-1}) {bUsePreviewValueAsDefault=true, inputName=vertex_0, inputType=FunctionInput_Vector2, sortPriority=1}) {name=Vertex0, nodeColor=<rgba 0 0 0 1>}
  NamedRerouteDeclaration(None=FunctionInput(Preview=Constant2Vector() {g=-1, r=1}) {bUsePreviewValueAsDefault=true, inputName=vertex_1, inputType=FunctionInput_Vector2, sortPriority=2}) {name=Vertex1, nodeColor=<rgba 0 0 0 1>}
  NamedRerouteDeclaration(None=FunctionInput(Preview=Constant2Vector() {g=1, r=1}) {bUsePreviewValueAsDefault=true, inputName=vertex_2, inputType=FunctionInput_Vector2, sortPriority=3}) {name=Vertex2, nodeColor=<rgba 0 0 0 1>}
  MaterialExpressionNamedRerouteUsage_12 = NamedRerouteUsage()
  MaterialExpressionNamedRerouteUsage_14 = NamedRerouteUsage()
  MaterialExpressionNamedRerouteUsage_18 = NamedRerouteUsage()
  MaterialExpressionNamedRerouteUsage_19 = NamedRerouteUsage()
  MaterialExpressionNamedRerouteUsage_28 = NamedRerouteUsage()
  n8 = Custom!(
   p=@MaterialExpressionNamedRerouteUsage_12
   v0=@MaterialExpressionNamedRerouteUsage_14
   v1=@MaterialExpressionNamedRerouteUsage_18
   v2=@MaterialExpressionNamedRerouteUsage_19
   v3=@MaterialExpressionNamedRerouteUsage_28
  ) {
   code="    const int num = 4;\r\n    float2 v[num] = { v0, v1, v2, v3 };\r\n    float d = dot(p-v[0],p-v[0]);\r\n    float s = 1.0;\r\n    float a_s = 0.;\r\n    for( int i=0, j=num-1; i<num; j=i, i++ )\r\n    {\r\n        // distance\r\n        float2 e = v[j] - v[i];\r\n        float2 w =    p - v[i];\r\n        float2 b = w - e*clamp( dot(w,e)/dot(e,e), 0.0, 1.0 );\r\n        d = min( d, dot(b,b) );\r\n\r\n        float cond1 = ceil(clamp((p.y-v[i].y),0.0,1.0)) + clamp(floor(1.-abs(p.y-v[i].y)), 0.0, 1.0);;\r\n        float cond2 = ceil(clamp((v[j].y-p.y),0.0,1.0));\r\n        float cond3 = ceil(clamp((e.x*w.y-e.y*w.x),0.0,1.0));\r\n      \r\n        a_s = cond1 + cond2 + cond3;\r\n        \r\n        float r1 = floor(a_s/3.);\r\n        float r2 = clamp(floor(1.-a_s),0.0,1.0);\r\n    \r\n    \r\n        if(r1+r2 >0.){\r\n            s=-s;\r\n            }\r\n    }\r\n    return s*sqrt(d);"
   inputs [
    {inputName=p, input {expression=&self:MaterialExpressionNamedRerouteUsage_12}}
    {inputName=v0, input {expression=&self:MaterialExpressionNamedRerouteUsage_14}}
    {inputName=v1, input {expression=&self:MaterialExpressionNamedRerouteUsage_18}}
    {inputName=v2, input {expression=&self:MaterialExpressionNamedRerouteUsage_19}}
    {inputName=v3, input {expression=&self:MaterialExpressionNamedRerouteUsage_28}}
   ]
  }
  NamedRerouteDeclaration(None=@n8) {name=SDF Shape, nodeColor=<rgba 0 0 0 1>}
  NamedRerouteDeclaration(None=FunctionInput(Preview=Constant2Vector() {g=1, r=-1}) {bUsePreviewValueAsDefault=true, inputName=vertex_3, inputType=FunctionInput_Vector2, sortPriority=4}) {name=Vertex3, nodeColor=<rgba 0 0 0 1>}
  n9 = NamedRerouteUsage()
  n10 = FunctionInput(Preview=Constant() {r=0.05}) {
   bUsePreviewValueAsDefault=true
   description=Determines the thickness of the stroke when using one of the outline outputs.
   inputName=Outline Thickness
   inputType=FunctionInput_Scalar
   sortPriority=9
  }
  NamedRerouteDeclaration(None=Subtract(A=Abs(None=@n9), B=@n10) {constB=0.1}) {name=SDF Outline, nodeColor=<rgba 0 0 0 1>}
  n11 = FunctionInput(Preview=Constant()) {
   bUsePreviewValueAsDefault=true
   description=Determines glow falloff strength. Can not go below GlowMin value or results will invert
   inputName=Glow Min
   inputType=FunctionInput_Scalar
   sortPriority=7
  }
  NamedRerouteDeclaration(None=@n11) {name=Glow Min, nodeColor=<rgba 0 0 0 1>}
  n12 = FunctionInput(Preview=Constant()) {
   bUsePreviewValueAsDefault=true
   description=Determines glow falloff strength. Can not go below GlowMin value or results will invert
   inputName=Glow Max
   inputType=FunctionInput_Scalar
   sortPriority=8
  }
  NamedRerouteDeclaration(None=@n12) {name=Glow Max, nodeColor=<rgba 0 0 0 1>}
 ]}
