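# Translucent UI material that procedurally draws the center reticle for the AloeCrouton weapon.
# SDF circles, radial segment masks and mirrored line strokes form three layers: segmented outer arcs, a middle circle with ticks, and a rotatable arc part.
# Named reroutes carry the layers into summed fill and opacity chains.
# Parameters drive the rotated part's visibility, angle, blur, thickness and gradient fade, and a panned sine builds a pulse mask.
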Material /AloeCroutonWeapon/Items/AloeCrouton/UI/Materials/M_UI_AloueCrouton_CenterReticle
graph {
 properties {blendMode=BLEND_Translucent, materialDomain=MD_UI, pixelDepthOffsetMode=PDOM_Legacy, usedShadingModels=Default Lit}
 nodes [
  n1 = NamedRerouteUsage()
  n2 = LinearInterpolate(B=NamedRerouteUsage(), Alpha=ScalarParameter() {parameterName=ShowRotatedPart}) {constA=1}
  Add(A=Add(A=NamedRerouteUsage(), B=Multiply(A=@n1, B=@n2)), B=NamedRerouteUsage())
  n3 = NamedRerouteUsage()
  n4 = MaterialFunctionCall!(UV, Mirror Y, Swizzle UV) {materialFunction=/FortUILibrary/MaterialFunctions/Basics/MF_UI_Symmetry_2x1_New}
  n5 = Constant() {r=0.0914}
  n6 = MaterialFunctionCall(Texture=MaterialFunctionCall(UVs=Add(A=@n4.Symmetry, B=@n5), Texture Scale=Constant() {r=0.5}) {materialFunction=&ScaleUVsByCenter}.UVs) {materialFunction=&MF_UI_LinearGradient}
  n7 = Multiply(
   A=LinearInterpolate(B=ScalarParameter() {defaultValue=0.5, parameterName=RotatedPartOpacity}, Alpha=ScalarParameter() {parameterName=ShowRotatedPart})
   B=LinearInterpolate(B=OneMinus(None=@n6.Result), Alpha=ScalarParameter() {parameterName=AddGradient}) {constA=1}
  )
  n8 = NamedRerouteUsage()
  n9 = ScalarParameter() {defaultValue=1, parameterName=CenterPartOpacity}
  Add(A=Add(A=Multiply(A=@n3, B=@n7), B=Multiply(A=@n8, B=@n9)), B=NamedRerouteUsage())
  n10 = MaterialFunctionCall(Size=Constant() {r=0.8}, OutlineThickness=Constant() {r=0.01}) {materialFunction=&MF_UI_Circle}
  n11 = MaterialFunctionCall!(UV, Segment Count=Constant() {r=2}, Segment Spacing=Constant() {r=0.3109}, Fixed Spacing) {materialFunction=/FortUILibrary/MaterialFunctions/Basics/MF_UI_RadialSegments}
  n12 = Multiply(
   A=OneMinus(None=SmoothStep(Value=@n10.SDFOutline) {constMax=0.03})
   B=OneMinus(None=SmoothStep(Value=@n11.Segmented Triangles Gradient) {constMax=0.03})
  )
  NamedRerouteDeclaration(None=@n12) {name=Fill, nodeColor=<rgba 0 0 0 1>}
  NamedRerouteDeclaration(None=@n12) {name=CenterReticleOpacity, nodeColor=<rgba 0 0 0 1>}
  n13 = ScalarParameter() {defaultValue=0.05, parameterName=RotatedPartBluriness}
  n14 = MaterialFunctionCall(Size=Constant() {r=0.6}, OutlineThickness=ScalarParameter() {defaultValue=0.0291, parameterName=RotatedPartOutlineThickness}) {materialFunction=&MF_UI_Circle}
  n15 = MaterialFunctionCall(
   UV=MaterialFunctionCall(UVs=TextureCoordinate(), "Rotation Angle (0-1)"=ScalarParameter() {defaultValue=0.25, parameterName=RotataionAngle}) {materialFunction=&CustomRotator}.Rotated Values
   Segment Count=Constant() {r=2}
   Segment Spacing=Constant() {r=0.1463}
  ) {materialFunction=&MF_UI_RadialSegments}
  n16 = Multiply(
   A=OneMinus(None=SmoothStep(Max=@n13, Value=@n14.SDFOutline) {constMax=0.03})
   B=OneMinus(None=SmoothStep(Max=@n13, Value=@n15.Segmented Triangles Gradient) {constMax=0.03})
  )
  NamedRerouteDeclaration(None=@n16) {name=RotatedPartFill, nodeColor=<rgba 0 0 0 1>}
  NamedRerouteDeclaration(None=@n16) {name=RotatedPartOpacity, nodeColor=<rgba 0 0 0 1>}
  n17 = MaterialFunctionCall(
   UVs=MaterialFunctionCall(UVs=TextureCoordinate(), Texture Scale=Constant() {r=0.4}) {materialFunction=&ScaleUVsByCenter}.UVs
   Size=Constant() {r=0.2}
   OutlineThickness=Constant() {r=0.01}
  ) {materialFunction=&MF_UI_Circle}
  n18 = MaterialFunctionCall(
   UVs=Add(
    A=MaterialFunctionCall!(UV, Swizzle UV) {materialFunction=/FortUILibrary/MaterialFunctions/Basics/MF_UI_Symmetry_2x2_New}.Symmetry
    B=Constant() {r=-0.5164}
   )
   "Rotation Angle (0-1)"=Constant() {r=0.1341}
  ) {materialFunction=&CustomRotator}
  n19 = Constant() {r=0.392}
  n20 = Add(
   A=OneMinus(None=SmoothStep(Value=@n17.SDFOutline) {constMax=0.03})
   B=MaterialFunctionCall!(
    Coordinates=MaterialFunctionCall(UVs=@n18.Rotated Values, Texture Scale=@n19) {materialFunction=&ScaleUVsByCenter}.UVs
    P0=Constant2Vector() {g=0.2, r=0.5}
    P1=Constant2Vector() {g=0.1, r=0.5}
    Line Width=Constant() {r=0.02}
    Edge Falloff=Constant() {r=0.8901}
    Square Corners?=StaticBool()
   ) {materialFunction=/Engine/Functions/Engine_MaterialFunctions01/Coordinates/DrawLine-2D}.Result
  )
  NamedRerouteDeclaration(None=@n20) {name=MiddleCircleFill, nodeColor=<rgba 0 0 0 1>}
  NamedRerouteDeclaration(None=@n20) {name=MIddleCircleOpacity, nodeColor=<rgba 0 0 0 1>}
  n21 = MaterialFunctionCall(
   UVs=MaterialFunctionCall() {materialFunction=&MF_UI_Symmetry_2x1_New}.Symmetry
   "Rotation Angle (0-1)"=Constant() {r=0.25}
  ) {materialFunction=&CustomRotator}
  n22 = Sine(None=ComponentMask(None=Panner(Coordinate=@n21.Rotated Values) {speedY=-1}) {g=true})
  NamedRerouteDeclaration(None=Add(A=Power(Base=@n22) {constExponent=20}) {constB=0.6}) {name=Panner, nodeColor=<rgba 0 0 0 1>}
 ]
}
