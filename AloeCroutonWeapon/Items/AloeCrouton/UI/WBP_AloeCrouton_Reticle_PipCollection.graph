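# Reticle widget for the AloeCrouton weapon showing stacked charge pips and a cooldown fill, built on the recharge stack pip collection base.
# Construct registers the dynamic entry box with the base, and UpdateSpoolCharge sets the fill material's Progress to one minus charge.
# AnimatePipCollection calls OnUse on the pip entry at index 1, then plays Anim_Shoot.
# That animation hides the real stack while stand-in pip images slide and fade, then fades it back in.
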WidgetBlueprint /AloeCroutonWeapon/Items/AloeCrouton/UI/WBP_AloeCrouton_Reticle_PipCollection
animations [Anim_Shoot {
  binding_count=4
  length_seconds=1
  object=&self:Anim_Shoot
  sequence {bindings [
    Pip_Empty {
     id=#1
     possessed_class=Image
     tracks [
      {
       type=MovieScene2DTransformTrack
       sections [{
         type=MovieScene2DTransformSection
         end_frame=10
         start_frame=0
         channels [
          Translation.X {type=MovieSceneScriptingFloatChannel, keys [[0,8.9],[10,23]]}
          Translation.Y {type=MovieSceneScriptingFloatChannel, keys [[0,47.3],[10,40.8]]}
          Angle {type=MovieSceneScriptingFloatChannel, const [64,2]}
         ]
        }]
      }
      {
       type=MovieSceneWidgetMaterialTrack
       sections [{type=MovieSceneParameterSection, end_frame=20, start_frame=0, channels [Progress {type=MovieSceneScriptingFloatChannel, const [1,3]}]}]
      }
      {
       type=MovieSceneFloatTrack
       display_name=Render Opacity
       sections [{type=MovieSceneFloatSection, end_frame=20, start_frame=0, channels [None {type=MovieSceneScriptingFloatChannel, keys [[0,1],[15,1],[20,0]]}]}]
      }
     ]
    }
    DynamicEntryBox_Stacks {
     id=#2
     possessed_class=DynamicEntryBox
     tracks [{
       type=MovieSceneFloatTrack
       display_name=Render Opacity
       sections [{type=MovieSceneFloatSection, end_frame=20, start_frame=0, channels [None {type=MovieSceneScriptingFloatChannel, keys [[0,0],[10,0],[15,1]]}]}]
      }]
    }
    Pip_Empty_NotMoving {
     id=#3
     possessed_class=Image
     tracks [{
       type=MovieSceneFloatTrack
       display_name=Render Opacity
       sections [{type=MovieSceneFloatSection, end_frame=20, start_frame=0, channels [None {type=MovieSceneScriptingFloatChannel, keys [[12,1],[15,0]]}]}]
      }]
    }
    Pip_Filled {
     id=#4
     possessed_class=Image
     tracks [
      {
       type=MovieScene2DTransformTrack
       sections [{
         type=MovieScene2DTransformSection
         end_frame=15
         start_frame=0
         channels [
          Translation.X {type=MovieSceneScriptingFloatChannel, keys [[0,23],[10,34.208]]}
          Translation.Y {type=MovieSceneScriptingFloatChannel, keys [[0,40.8],[10,33.0457]]}
          Angle {type=MovieSceneScriptingFloatChannel, keys [[0,64],[10,45.5626]]}
         ]
        }]
      }
      {
       type=MovieSceneWidgetMaterialTrack
       sections [{
         type=MovieSceneParameterSection
         end_frame=20
         start_frame=0
         channels [
          AR {type=MovieSceneScriptingFloatChannel, const [-0.5,1]}
          AG {type=MovieSceneScriptingFloatChannel, keys [[0,-0.6],[10,0.59]]}
          AB {type=MovieSceneScriptingFloatChannel, const [0,1]}
          AA {type=MovieSceneScriptingFloatChannel, const [1,1]}
          BR {type=MovieSceneScriptingFloatChannel, const [0.5,1]}
          BG {type=MovieSceneScriptingFloatChannel, keys [[0,-0.8],[10,0.78]]}
          BB {type=MovieSceneScriptingFloatChannel, const [0,1]}
          BA {type=MovieSceneScriptingFloatChannel, const [1,1]}
         ]
        }]
      }
      {
       type=MovieSceneFloatTrack
       display_name=Render Opacity
       sections [{
         type=MovieSceneFloatSection
         end_frame=20
         start_frame=0
         channels [None {type=MovieSceneScriptingFloatChannel, keys [[0,1],[10,1],[15,0.6],[20,0]]}]
        }]
      }
     ]
    }
   ]}
 }]
graphs [
 AnimatePipCollection {nodes [
   AnimatePipCollection() {is=title K2Node_FunctionEntry}
   n1 = GetDynamicEntryBox!(self:Fort Weapon Reticle Recharge Stack Base Pip Collection Base Object Reference) -> (ReturnValue:Dynamic Entry Box Object Reference)
   > n2 = Cast To WBP_AloeCrouton_Reticle_Pip!(
    execute
    Object="Get (a copy)"(
     Array:Array of User Widget Object References=GetAllEntries!(self:Dynamic Entry Box Base Object Reference=@n1) -> (ReturnValue:Array of User Widget Object References)
     Dimension 1=1
    ) -> (Output:User Widget Object Reference)
   ) -> (then, CastFailed, AsWBP Aloe Crouton Reticle Pip:WBP Aloe Crouton Reticle Pip Object Reference) {is=title K2Node_DynamicCast}
   > n3 = Set PipRef(PipRef:WBP Aloe Crouton Reticle Pip Object Reference=@n2) -> (Output_Get:WBP Aloe Crouton Reticle Pip Object Reference)
   > OnUse(self:WBP Reticle Recharge Stack Base Pip Object Reference=@n3)
   > PlayAnimation(InAnimation=Get Anim_Shoot() -> (Anim_Shoot:Widget Animation Object Reference), PlaybackSpeed=1.2)
  ]}
 EventGraph {nodes [
   Event Construct()
   > SetDynamicEntryBox!(
    execute
    self:Fort Weapon Reticle Recharge Stack Base Pip Collection Base Object Reference
    DynamicEntryBox:Dynamic Entry Box Object Reference=Get DynamicEntryBox_Stacks() -> (DynamicEntryBox_Stacks:Dynamic Entry Box Object Reference)
   ) -> (then)
   Event PreConstruct()
   Event Tick()
  ]}
 GenerateStacks {nodes [
   n1 = GenerateStacks!() -> (then, TotalStacks:int=0, CurrentStacks:int=0) {is=title K2Node_FunctionEntry}
   > Parent: GenerateStacks!(execute, TotalStacks:int=@n1, CurrentStacks:int=@n1.CurrentStacks) -> (then) {is=title K2Node_CallParentFunction}
  ]}
 UpdateSpoolCharge {nodes [
   n1 = UpdateSpoolCharge!() -> (then, InChargePercent:float=~) {is=title K2Node_FunctionEntry}
   > n2 = GetDynamicMaterial(self=Get Image_ProgressFill() -> (Image_ProgressFill:Image Object Reference))
   > SetScalarParameterValue(self=@n2, ParameterName=Progress, Value=float - float(A=1, B=@n1))
  ]}
]
variables []
widgets {
 info {
  inheritedWidgetCount=0
  parentClass=/Game/Athena/HUD/reticles/ReticleExtensionWidgets/ChargeStack/WBP_Reticle_RechargeStackBase_PipCollection.WBP_Reticle_RechargeStackBase_PipCollection_C
  rootWidgetClass=&SizeBox
  widgetCount=7
 }
 widgets [
  SizeBox_0 {
   widgetClassPath=&SizeBox
   widget=SizeBox_0
   properties {bOverride_HeightOverride=true, bOverride_WidthOverride=true, heightOverride=128, widthOverride=128}
  }
  Overlay_0 {widgetClassPath=&Overlay, parent=SizeBox_0, widget=Overlay_0, slot {class=SizeBoxSlot}}
  DynamicEntryBox_Stacks {
   widgetClassPath=&DynamicEntryBox
   bIsVariable=true
   parent=Overlay_0
   widget=DynamicEntryBox_Stacks
   properties {
    entryBoxType=Vertical
    entryWidgetClass=&/WBP_AloeCrouton_Reticle_Pip.WBP_AloeCrouton_Reticle_Pip_C
    numDesignerPreviewEntries=2
    visibility=HitTestInvisible
    entrySpacing=<0 1.5676>
    radialBoxSettings {angleBetweenItems=14.8114, bClockwiseOrder=false, bDistributeItemsEvenly=false, marginSize=0, sectorCentralAngle=360, startingAngle=279.36}
    renderTransform {angle=65, scale=<1 1>, shear=<0 0>, translation=<16 44>}
   }
   slot {class=OverlaySlot, properties {horizontalAlignment=HAlign_Center, verticalAlignment=VAlign_Center}}
  }
  Image_ProgressFill {
   widgetClassPath=&Image
   bIsVariable=true
   parent=Overlay_0
   widget=Image_ProgressFill
   properties {
    brush {
     bIsDynamicallyLoaded=false
     drawAs=Image
     imageType=NoImage
     mirroring=NoMirror
     resourceName=None
     resourceObject=/AloeCroutonWeapon/Items/AloeCrouton/UI/Materials/MI_UI_AloeCrouton_Cooldown_Main
     tiling=NoTile
     imageSize=<64 64>
     margin=<ltrb 0 0 0 0>
     outlineSettings {
      bUseBrushTransparency=false
      roundingType=HalfHeightRadius
      width=0
      color {colorUseRule=UseColor_Specified, specifiedColor=<rgba 0 0 0 0>}
      cornerRadii=<0 0 0 0>
     }
     tintColor {colorUseRule=UseColor_Specified, specifiedColor=<rgba 1 1 1 1>}
     uVRegion {bIsValid=false, max=<0 0>, min=<0 0>}
    }
    renderTransform {angle=-21, scale=<1.2 1.2>, shear=<0 0>, translation=<7 6>}
   }
   slot {class=OverlaySlot, properties {horizontalAlignment=HAlign_Fill, verticalAlignment=VAlign_Fill}}
  }
  Pip_Empty_NotMoving {
   widgetClassPath=&Image
   bIsVariable=true
   parent=Overlay_0
   widget=Pip_Empty_NotMoving
   properties {
    brush {
     bIsDynamicallyLoaded=false
     drawAs=Image
     imageType=NoImage
     mirroring=NoMirror
     resourceName=None
     resourceObject=/AloeCroutonWeapon/Items/AloeCrouton/UI/Materials/M_UI_AloeCrouton_Pip_Empty
     tiling=NoTile
     imageSize=<12 14>
     margin=<ltrb 0 0 0 0>
     outlineSettings {
      bUseBrushTransparency=false
      roundingType=HalfHeightRadius
      width=0
      color {colorUseRule=UseColor_Specified, specifiedColor=<rgba 0 0 0 0>}
      cornerRadii=<0 0 0 0>
     }
     tintColor {colorUseRule=UseColor_Specified, specifiedColor=<rgba 1 1 1 1>}
     uVRegion {bIsValid=false, max=<0 0>, min=<0 0>}
    }
    renderTransform {angle=64, scale=<1 1>, shear=<0 0>, translation=<8.9 47.3>}
   }
   slot {class=OverlaySlot, properties {horizontalAlignment=HAlign_Center, verticalAlignment=VAlign_Center}}
  }
  Pip_Empty {
   widgetClassPath=&Image
   bIsVariable=true
   parent=Overlay_0
   widget=Pip_Empty
   properties {
    brush {
     bIsDynamicallyLoaded=false
     drawAs=Image
     imageType=NoImage
     mirroring=NoMirror
     resourceName=None
     resourceObject=/AloeCroutonWeapon/Items/AloeCrouton/UI/Materials/MI_UI_AloeCrouton_Pip
     tiling=NoTile
     imageSize=<12 14>
     margin=<ltrb 0 0 0 0>
     outlineSettings {
      bUseBrushTransparency=false
      roundingType=HalfHeightRadius
      width=0
      color {colorUseRule=UseColor_Specified, specifiedColor=<rgba 0 0 0 0>}
      cornerRadii=<0 0 0 0>
     }
     tintColor {colorUseRule=UseColor_Specified, specifiedColor=<rgba 1 1 1 1>}
     uVRegion {bIsValid=false, max=<0 0>, min=<0 0>}
    }
    renderTransform {angle=64, scale=<1 1>, shear=<0 0>, translation=<8.9 47.3>}
   }
   slot {class=OverlaySlot, properties {horizontalAlignment=HAlign_Center, verticalAlignment=VAlign_Center}}
  }
  Pip_Filled {
   widgetClassPath=&Image
   bIsVariable=true
   parent=Overlay_0
   widget=Pip_Filled
   properties {
    brush {
     bIsDynamicallyLoaded=false
     drawAs=Image
     imageType=NoImage
     mirroring=NoMirror
     resourceName=None
     resourceObject=&MI_UI_AloeCrouton_Pip
     tiling=NoTile
     imageSize=<12 14>
     margin=<ltrb 0 0 0 0>
     outlineSettings {
      bUseBrushTransparency=false
      roundingType=HalfHeightRadius
      width=0
      color {colorUseRule=UseColor_Specified, specifiedColor=<rgba 0 0 0 0>}
      cornerRadii=<0 0 0 0>
     }
     tintColor {colorUseRule=UseColor_Specified, specifiedColor=<rgba 1 1 1 1>}
     uVRegion {bIsValid=false, max=<0 0>, min=<0 0>}
    }
    renderTransform {angle=64, scale=<1 1>, shear=<0 0>, translation=<23 40.8>}
   }
   slot {class=OverlaySlot, properties {horizontalAlignment=HAlign_Center, verticalAlignment=VAlign_Center}}
  }
 ]
}
