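# Reticle widget for the AloeCrouton weapon, extending the recharge-stack reticle base with a charge-animated center image and a pip collection.
# Weapon start/end charge delegates drive the center's dynamic material, animating pips if charges are full at start.
# An async wait on the Charge attribute refreshes stack count and pip colors.
# A cooldown timer, active below max charges, updates recharge progress, while UpdateSpoolCharge passes inverted charge percent to the pips.
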WidgetBlueprint /AloeCroutonWeapon/Items/AloeCrouton/UI/WBP_AloeCrouton_Reticle
animations [
 Intro {
  binding_count=1
  length_seconds=0.3
  object=&self:Intro
  sequence {bindings [Reticle {
     id=#1
     possessed_class=Overlay
     tracks [
      {
       type=MovieSceneFloatTrack
       display_name=Render Opacity
       sections [{type=MovieSceneFloatSection, channels [None {type=MovieSceneScriptingFloatChannel, keys [[0,0],[6,1]]}]}]
      }
      {
       type=MovieScene2DTransformTrack
       sections [{
         type=MovieScene2DTransformSection
         channels [
          Translation.X {type=MovieSceneScriptingFloatChannel}
          Translation.Y {type=MovieSceneScriptingFloatChannel}
          Angle {type=MovieSceneScriptingFloatChannel}
          Scale.X {type=MovieSceneScriptingFloatChannel, keys [[0,1.25],[6,1]]}
          Scale.Y {type=MovieSceneScriptingFloatChannel, keys [[0,1.25],[6,1]]}
          Shear.X {type=MovieSceneScriptingFloatChannel}
          Shear.Y {type=MovieSceneScriptingFloatChannel}
         ]
        }]
      }
     ]
    }]}
 }
 Anim_CenterReticleCharge {
  binding_count=1
  length_seconds=1
  object=&self:Anim_CenterReticleCharge
  sequence {bindings [ReticleVisual {
     id=#2
     possessed_class=Image
     tracks [{
       type=MovieSceneWidgetMaterialTrack
       sections [{
         type=MovieSceneParameterSection
         end_frame=20
         start_frame=0
         channels [
          ShowRotatedPart {type=MovieSceneScriptingFloatChannel, keys [[0,0],[6,1]]}
          RotataionAngle {type=MovieSceneScriptingFloatChannel, keys [[6,0.25],[16,1]]}
          RotatedPartOutlineThickness {type=MovieSceneScriptingFloatChannel, keys [[16,0.0291],[20,0.0691]]}
          RotatedPartBluriness {type=MovieSceneScriptingFloatChannel, keys [[16,0.05],[20,0.13]]}
          AddGradient {type=MovieSceneScriptingFloatChannel, keys [[6,0],[16,1]]}
         ]
        }]
      }]
    }]}
 }
 Anim_HideRotatedPart {
  binding_count=1
  length_seconds=0.25
  object=&self:Anim_HideRotatedPart
  sequence {bindings [ReticleVisual {
     id=#3
     possessed_class=Image
     tracks [{
       type=MovieSceneWidgetMaterialTrack
       sections [{
         type=MovieSceneParameterSection
         end_frame=5
         start_frame=0
         channels [ShowRotatedPart {type=MovieSceneScriptingFloatChannel, keys [[0,1],[5,0]]}]
        }]
      }]
    }]}
 }
]
graphs [
 BindToWeaponEvents {nodes [
   n1 = BindToWeaponEvents!() -> (then, DoBind:bool, EquippedWeapon:Fort Weapon Object Reference) {is=title K2Node_FunctionEntry}
   > Parent: BindToWeaponEvents!(execute, DoBind:bool=@n1, EquippedWeapon:Fort Weapon Object Reference=@n1.EquippedWeapon) -> (then) {is=title K2Node_CallParentFunction}
   > n2 = Cast To B_AloeCrouton_Athena!(execute, Object=@n1.EquippedWeapon) -> (then, CastFailed, AsB Aloe Crouton Athena:B Aloe Crouton Athena Object Reference) {is=title K2Node_DynamicCast}
   > n3 = Set AloeCrouton Wep Ref(AloeCrouton Wep Ref:B Aloe Crouton Athena Object Reference=@n2) -> (Output_Get:B Aloe Crouton Athena Object Reference)
   > Bind Event to Update UIOverheat Visuals(self:B Aloe Crouton Athena Object Reference=@n3, Delegate=Create Event()) {is=title K2Node_AddDelegate}
  ]}
 EventGraph {nodes [
   n1 = StartCharge() {is=title K2Node_CustomEvent}
   > n2 = GetDynamicMaterial(self=Get ReticleVisual() -> (ReticleVisual:Image Object Reference))
   > SetScalarParameterValue(self=@n2, ParameterName=RotatedPartOpacity, Value=0.5)
   > PlayAnimationForward(InAnimation=Get Anim_CenterReticleCharge() -> (Anim_CenterReticleCharge:Widget Animation Object Reference), PlaybackSpeed=0.5)
   > n3 = Get!(execute:Exec) -> (then:Exec, else:Exec, AloeCrouton Wep Ref:B Aloe Crouton Athena Object Reference)
   > Branch(Condition="Equal (Float)"(
     A=GetMaxChargeCount!(self:B Aloe Crouton Athena Object Reference=@n3) -> (MaxChargeCount:double)
     B=GetChargeCount!(self:B Aloe Crouton Athena Object Reference=@n3) -> (ChargeCount:double)
    ))
   > AnimatePipCollection(self:WBP Aloe Crouton Reticle Pip Collection Object Reference=Get WBP_AloeCrouton_Reticle_PipCollection() -> (WBP_AloeCrouton_Reticle_PipCollection:WBP Aloe Crouton Reticle Pip Collection Object Reference))
   n4 = EndedCharge() {is=title K2Node_CustomEvent}
   > n5 = GetDynamicMaterial(self=Get ReticleVisual())
   > SetScalarParameterValue(self=@n5, ParameterName=RotatedPartOpacity, Value=0)
   Event UpdateCooldownQueryTimer() {is=title K2Node_Event}
   > n6 = Get()
   > n7 = Branch(Condition="float >= float"(A=GetChargeCount(self=@n6), B=GetMaxChargeCount(self=@n6)))
   > HandleCooldownTimer!(execute, self:Fort Weapon Reticle Extension Widget Base Object Reference, bActivateTimer:bool=false, TimerInterval:float=0.1) -> (then)
   HandleCooldownTimer(execute=@n7.else, bActivateTimer=true, TimerInterval=0.1)
   Event Construct()
   > Parent: Construct()
   > n8 = Sequence!(execute) -> (then_0, then_1, then_2:Exec)
   > PlayAnimation(InAnimation=Get Intro())
   > n9 = Get()
   > GenerateStacks!(
    execute
    self:WBP Reticle Recharge Stack Base Pip Collection Object Reference=Get WBP_AloeCrouton_Reticle_PipCollection()
    TotalStacks:int=Truncate(A=GetMaxChargeCount(self=@n9))
    CurrentStacks:int=Truncate(A=GetChargeCount(self=@n9))
   ) -> (then)
   n10 = Get(execute=@n8.then_2)
   > Bind Event to Update UI on Start Charge(self:B Aloe Crouton Athena Object Reference=@n10, Delegate=@n1) {is=title K2Node_AddDelegate}
   > Bind Event to Update UI on End Charge(self:B Aloe Crouton Athena Object Reference=@n10, Delegate=@n4) {is=title K2Node_AddDelegate}
   n11 = WaitForAttributeChanged!(
    execute:Exec=@n8.then_1
    TargetActor:Actor Object Reference=GetOwningPlayerPawn!(self:User Widget Object Reference) -> (ReturnValue:Pawn Object Reference)
    Attribute:Gameplay Attribute Structure=(AttributeName="Charge",Attribute=/Script/AloeCroutonWeaponRuntime.FortChargingSet_AloeCrouton:Charge,AttributeOwner="&Class'&FortChargingSet_AloeCrouton'")
    OnlyTriggerOnce:bool=false
   ) -> (
    then:Exec
    AsyncTaskProxy:Ability Async Wait Attribute Changed Object Reference
    Changed:Exec
    Attribute:Gameplay Attribute Structure
    NewValue:float=~
    OldValue:float=~
   ) {is=title K2Node_AsyncAction}
   n12 = Get(execute=[@n11.Changed, @n11.then])
   > UpdateStackCount!(
    execute
    self:WBP Reticle Recharge Stack Base Pip Collection Object Reference=Get WBP_AloeCrouton_Reticle_PipCollection()
    NewCount:int=Truncate(A=GetChargeCount(self=@n12))
   ) -> (then)
   > UpdateCooldownQueryTimer(self:Fort Weapon Reticle Extension Widget Base Object Reference)
   n13 = GetMaxChargeCount(self=@n12)
   > n14 = For Loop(LastIndex=int - int(A=Truncate(A=@n13)))
   > SetPipColor!(
    execute
    self:Fort Weapon Reticle Recharge Stack Base Pip Collection Base Object Reference=Get WBP_AloeCrouton_Reticle_PipCollection()
    Progress:float=1
    Index:int=@n14
    NotReadyColor:Linear Color Structure=Get NotReadyColor() -> (NotReadyColor:Linear Color Structure)
    ReadyColor:Linear Color Structure=Get ReadyColor() -> (ReadyColor:Linear Color Structure)
    PipColorMatParamName:Name=PipColor
   ) -> (then)
   Event Destruct()
   > Parent: Destruct()
   > HandleCooldownTimer(bActivateTimer=false, TimerInterval=0)
   Event UpdateRemainingCooldown() {is=title K2Node_Event}
   > n15 = GetRemainingCooldownTimer!(execute, self) -> (then, RemainingTime:double, TotalDuration:double)
   > UpdateRemainingRechargeTime!(
    execute
    self:Fort Weapon Reticle Recharge Stack Base Pip Collection Base Object Reference=Get WBP_AloeCrouton_Reticle_PipCollection()
    TimeRemaining:float=@n15
    TotalDuration:float=@n15.TotalDuration
    PipPercentMatParamName:Name=Progress
    ShouldSetPipColor:bool=true
    NotReadyColor:Linear Color Structure=Get NotReadyColor()
    ReadyColor:Linear Color Structure=Get ReadyColor()
    PipColorMatParamName:Name=PipColor
   ) -> (then)
  ]}
 GetPipCollection {nodes [
   GetPipCollection() {is=title K2Node_FunctionEntry}
   > Return Node!(execute, PipCollection:WBP Reticle Recharge Stack Base Pip Collection Object Reference=Get WBP_AloeCrouton_Reticle_PipCollection())
  ]}
 GetRemainingCooldownTimer {nodes [
   GetRemainingCooldownTimer() {is=title K2Node_FunctionEntry}
   > n1 = Get()
   > Return Node!(
    execute
    RemainingTime:double=GetCooldownTimeRemaining!(self:B Aloe Crouton Athena Object Reference=@n1) -> (CooldownTimeRemaining:double)
    TotalDuration:double=GetCooldownDuration!(self:B Aloe Crouton Athena Object Reference=@n1) -> (CooldownDuration:double)
   )
  ]}
 UpdateSpoolCharge {nodes [
   n1 = UpdateSpoolCharge!() -> (then, ChargePercent:double=~) {is=title K2Node_FunctionEntry}
   > UpdateSpoolCharge!(
    execute
    self:WBP Aloe Crouton Reticle Pip Collection Object Reference=Get WBP_AloeCrouton_Reticle_PipCollection()
    InChargePercent:float=float - float(A=1, B=@n1)
   ) -> (then) {is=title K2Node_CallFunction}
  ]}
]
variables [
 AloeCrouton Wep Ref {type=B_AloeCrouton_Athena_C, default=null}
 NotReadyColor {type=LinearColor, default=<rgba 1 0.3763 0 1>}
 ReadyColor {type=LinearColor, default=<rgba 0.0006 0.9387 0.9647 1>}
]
widgets {
 info {
  inheritedWidgetCount=1
  parentClass=/Game/Athena/HUD/reticles/ReticleExtensionWidgets/ChargeStack/WBP_Reticle_RechargeStackBase.WBP_Reticle_RechargeStackBase_C
  rootWidgetClass=&Overlay
  widgetCount=5
 }
 widgets [
  Reticle {widgetClassPath=&Overlay, widget=Reticle}
  ReticleVisual {
   widgetClassPath=&Image
   bIsVariable=true
   parent=Reticle
   widget=ReticleVisual
   properties {
    brush {
     bIsDynamicallyLoaded=false
     drawAs=Image
     imageType=NoImage
     mirroring=NoMirror
     resourceName=None
     resourceObject=/AloeCroutonWeapon/Items/AloeCrouton/UI/Materials/MI_UI_AloueCrouton_CenterReticle
     tiling=NoTile
     imageSize=<128 128>
     margin=<ltrb 0 0 0 0>
     outlineSettings {
      bUseBrushTransparency=false
      roundingType=HalfHeightRadius
      width=0
      color {colorUseRule=UseColor_Specified, specifiedColor=<rgba 0 0 0 0>}
      cornerRadii=<0 0 0 0>
     }
     tintColor {colorUseRule=UseColor_Specified, specifiedColor=<rgba 1 1 1 1>}
     uVRegion {bIsValid=false, max=<0 0>, min=<0 0>}
    }
    renderTransform {angle=0, scale=<0.8 0.8>, shear=<0 0>, translation=<0 0>}
   }
   slot {class=OverlaySlot, properties {horizontalAlignment=HAlign_Center, verticalAlignment=VAlign_Center}}
  }
  FortVisualAttachment_0 {
   widgetClassPath=&FortVisualAttachment
   parent=Reticle
   widget=FortVisualAttachment_0
   properties {contentAnchor=<0 0.2>}
   slot {class=OverlaySlot}
  }
  WBP_AloeCrouton_Reticle_PipCollection {
   widgetClassPath=&/WBP_AloeCrouton_Reticle_PipCollection
   bIsVariable=true
   parent=FortVisualAttachment_0
   widget=WBP_AloeCrouton_Reticle_PipCollection
   properties {renderTransform {angle=-12, scale=<1 1>, shear=<0 0>, translation=<14 23>}}
   slot {class=SizeBoxSlot}
  }
  TutorialHighlight {widgetClassPath=&TutorialHighlight}
 ]
}
